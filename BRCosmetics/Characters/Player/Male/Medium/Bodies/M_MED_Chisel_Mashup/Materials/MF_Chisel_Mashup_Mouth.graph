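# Material function for the Chisel Mashup character's mouth, choosing a mouth shape from facial pose parameters via World Position Offset.
# Chained threshold If and Max nodes turn lip, smile, frown, jaw-open and OO/MBP phoneme poses into one vertex value.
# VertexColorMask compares that value against blue vertex color to toggle mouth elements.
# Its result scales each vertex's object-relative position, which is added to the incoming World Position Offset.
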MaterialFunction /BRCosmetics/Characters/Player/Male/Medium/Bodies/M_MED_Chisel_Mashup/Materials/MF_Chisel_Mashup_Mouth
graph {nodes [
  n1 = GetMaterialAttributes(None=FunctionInput() {inputType=FunctionInput_MaterialAttributes}) {attributeGetTypes [#1]}
  n2 = Subtract(A=WorldPosition().XYZ, B=ObjectPositionWS())
  n3 = ScalarParameter() {defaultValue=0.2, group=Banana, parameterName=THreshold}
  n4 = ScalarParameter() {group=Banana, parameterName=R_Frown_pose}
  n5 = ScalarParameter() {defaultValue=0.8, group=Banana, parameterName=Mouth Open Frown Vertex Value}
  n6 = ScalarParameter() {defaultValue=0.1, group=Banana, parameterName=Default Open Mouth Vertex Value}
  n7 = ScalarParameter() {group=Banana, parameterName=R_smile_pose}
  n8 = ScalarParameter() {defaultValue=0.4, group=Banana, parameterName=Mouth Open Smile Vertex Value}
  n9 = ScalarParameter() {defaultValue=0.7, group=Banana, parameterName=Default Face Vertex Value}
  n10 = If(
   A=ScalarParameter() {parameterName=L_upper_lip_raiser_pose}
   B=@n3
   "A > B"=ScalarParameter() {defaultValue=0.5, group=Banana, parameterName=Raise L Lip Corner Vertex Value}
   "A == B"=@n9
   "A < B"=@n9
  )
  n11 = Max(
   A=@n10
   B=If(
    A=ScalarParameter() {parameterName=R_upper_lip_raiser_pose}
    B=@n3
    "A > B"=ScalarParameter() {defaultValue=0.6, group=Banana, parameterName=Raise R Lip Corner Vertex Value}
    "A == B"=@n10
    "A < B"=@n10
   )
  )
  n12 = If(
   A=Max(A=ScalarParameter() {parameterName=R_lip_corner_narrow_pose}, B=ScalarParameter() {parameterName=PeelyBlenderFace})
   B=@n3
   "A > B"=ScalarParameter() {defaultValue=0.7, group=Banana, parameterName=Narrow Vertex Value}
   "A == B"=@n11
   "A < B"=@n11
  )
  n13 = If(
   A=ScalarParameter() {group=Banana, parameterName=R_frown_pose}
   B=@n3
   "A > B"=ScalarParameter() {defaultValue=0.3, group=Banana, parameterName=Frown Frame Vertex Value}
   "A == B"=@n12
   "A < B"=@n12
  )
  n14 = Max(
   A=@n13
   B=If(
    A=ScalarParameter() {group=Banana, parameterName=R_smile_pose}
    B=@n3
    "A > B"=ScalarParameter() {defaultValue=0.2, group=Banana, parameterName=Smile Mouth animation Vertex Value}
    "A == B"=@n13
    "A < B"=@n13
   )
  )
  n15 = If(
   A=ScalarParameter() {group=Banana, parameterName=Jaw_open_Pose}
   B=@n3
   "A > B"=Max(A=If(A=@n4, B=@n3, "A > B"=@n5, "A == B"=@n6, "A < B"=@n6), B=If(A=@n7, B=@n3, "A > B"=@n8, "A == B"=@n6, "A < B"=@n6))
   "A == B"=@n14
   "A < B"=@n14
  )
  n16 = If(
   A=ScalarParameter() {parameterName=phoneme_oo_pose}
   B=ScalarParameter() {defaultValue=0.05, group=Banana, parameterName=THreshold}
   "A > B"=ScalarParameter() {defaultValue=0.1, group=Banana, parameterName=Phoneme OO Vertex Value}
   "A == B"=@n15
   "A < B"=@n15
  )
  n17 = If(
   A=ScalarParameter() {parameterName=phoneme_mbp_pose}
   B=ScalarParameter() {defaultValue=0.05, group=Banana, parameterName=THreshold}
   "A > B"=ScalarParameter() {defaultValue=0.9, group=Banana, parameterName=Phoneme MBP Vertex Value}
   "A == B"=@n16
   "A < B"=@n16
  )
  n18 = Multiply(
   A=@n2
   B=MaterialFunctionCall!(Hide Element Toggle=Constant() {r=1}, VertexColorChannel=VertexColor().B, Vertex Value to hide=Max(A=@n16, B=@n17)) {materialFunction=/Game/Packages/Fortress_SharedMaterials/Base_Material_Functions/VertexColorMask}.Result
  )
  FunctionOutput(None=SetMaterialAttributes!(MaterialAttributes=@n1.MaterialAttributes, WorldPositionOffset=Add(A=@n1.WorldPositionOffset, B=@n18) {constB=0}) {attributeSetTypes [#1]})
  Multiply(A=@n2) {constB=-1}
  ScalarParameter() {defaultValue=0.1, parameterName=Test}
 ]}
