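# Physics asset for the face accessory mesh of the male medium Chisel Mashup character body, covering the dynamic "head peel" bones on each side.
# The left peel is a three-capsule chain and the right a sphere plus longer capsule, linked by angle-limited constraints with tight twist and wider primary swing on the right.
# Four capsules on spine_05 form a compound upper-body collision shape, with no constraint joining it to either chain.
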PhysicsAsset /BRCosmetics/Characters/Player/Male/Medium/Bodies/M_MED_Chisel_Mashup/Meshes/Parts/M_MED_Chisel_Mashup_Face_Acc_PhysicsAsset
bodies {
 head_peel_l_01_dyn [{shapeType=Capsule, length=5, radius=3, shapeName=head_peel_l_01_dyn_capsule, rotation=<rot 90 0 0>}]
 head_peel_l_02_dyn [{shapeType=Capsule, length=4, radius=3, shapeName=head_peel_l_02_dyn_capsule, rotation=<rot 90 0 0>}]
 head_peel_l_03_dyn [{shapeType=Capsule, length=5, radius=3, shapeName=head_peel_l_03_dyn_capsule, center=<4 0 0>, rotation=<rot 90 0 0>}]
 head_peel_r_01_dyn [{shapeType=Sphere, radius=1, shapeName=head_peel_r_01_dyn_sphere}]
 head_peel_r_02_dyn [{shapeType=Capsule, length=12, radius=2.9, shapeName=head_peel_r_02_dyn_capsule, center=<4 0.5 0>, rotation=<rot 90 0 0>}]
 spine_05 [
  {shapeType=Capsule, length=30, radius=6, shapeName=spine_05_capsule, center=<14.5 2.5 -0.5>, rotation=<rot 90 9 0>}
  {shapeType=Capsule, length=20, radius=6, shapeName=spine_05_capsule, center=<50 15 0>, rotation=<rot 90 -180 160>}
  {shapeType=Capsule, length=8, radius=6, shapeName=spine_05_capsule, center=<33 8 -17>, rotation=<rot 25 0 0>}
  {shapeType=Capsule, length=6, radius=4, shapeName=spine_05_capsule, center=<34.3966 8 -16.6774>, rotation=<rot 75 179.9998 180>}
 ]
}
constraints [
 {bone1Name=head_peel_r_02_dyn, bone2Name=head_peel_r_01_dyn, swing1LimitDegrees=45, swing2LimitDegrees=10, twistLimitDegrees=3}
 {bone1Name=head_peel_l_02_dyn, bone2Name=head_peel_l_01_dyn, swing1LimitDegrees=25, swing2LimitDegrees=10, twistLimitDegrees=3}
 {bone1Name=head_peel_l_03_dyn, bone2Name=head_peel_l_02_dyn, swing1LimitDegrees=25, swing2LimitDegrees=10, twistLimitDegrees=3}
]
